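# Material function in the ClockGear_SE wheel cosmetics that fakes an infinite-depth effect by stacking sixteen parallax layers of a texture, a dot mask by default.
# Each BumpOffset feeds the next, so offsets accumulate, and a static switch can route each layer's UVs through a Rotator.
# An optional gradient falloff weights layers with constants that halve at each step, alongside a darkness input.
# All layers are summed and clamped into a single output.
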MaterialFunction /VehicleCosmetics/Wheels/ClockGear_SE/Materials/MF_Infinity_OffsetFunction
graph {nodes [
  n1 = BumpOffset(Coordinate=NamedRerouteUsage(), Height=NamedRerouteUsage())
  n2 = TextureSample(UVs=StaticSwitch(True=Rotator(Coordinate=@n1), False=@n1, Value=NamedRerouteUsage()), Tex=NamedRerouteUsage()) {texture=/VehicleCosmetics/Wheels/ClockGear_SE/Textures/DotMask_01}
  n3 = NamedRerouteUsage()
  n4 = StaticSwitch(True=Multiply(A=@n2.RGBA, B=Multiply(B=@n3) {constA=327.68, desc="327,679993"}), False=@n2.RGBA, Value=NamedRerouteUsage())
  n5 = BumpOffset(Coordinate=@n1, Height=NamedRerouteUsage())
  n6 = TextureSample(UVs=StaticSwitch(True=Rotator(Coordinate=@n5), False=@n5, Value=NamedRerouteUsage()), Tex=NamedRerouteUsage()) {texture=&DotMask_01}
  n7 = NamedRerouteUsage()
  n8 = StaticSwitch(True=Multiply(A=@n6.RGBA, B=Multiply(B=@n7) {constA=163.84, desc=163.839996}), False=@n6.RGBA, Value=NamedRerouteUsage())
  n9 = BumpOffset(Coordinate=@n5, Height=NamedRerouteUsage())
  n10 = TextureSample(UVs=StaticSwitch(True=Rotator(Coordinate=@n9), False=@n9, Value=NamedRerouteUsage()), Tex=NamedRerouteUsage()) {texture=&DotMask_01}
  n11 = NamedRerouteUsage()
  n12 = StaticSwitch(True=Multiply(A=@n10.RGBA, B=Multiply(B=@n11) {constA=81.92, desc=81.92}), False=@n10.RGBA, Value=NamedRerouteUsage())
  n13 = BumpOffset(Coordinate=@n9, Height=NamedRerouteUsage())
  n14 = TextureSample(UVs=StaticSwitch(True=Rotator(Coordinate=@n13), False=@n13, Value=NamedRerouteUsage()), Tex=NamedRerouteUsage()) {texture=&DotMask_01}
  n15 = NamedRerouteUsage()
  n16 = StaticSwitch(True=Multiply(A=@n14.RGBA, B=Multiply(B=@n15) {constA=40.96, desc=40.96}), False=@n14.RGBA, Value=NamedRerouteUsage())
  n17 = BumpOffset(Coordinate=@n13, Height=NamedRerouteUsage())
  n18 = TextureSample(UVs=StaticSwitch(True=Rotator(Coordinate=@n17), False=@n17, Value=NamedRerouteUsage()), Tex=NamedRerouteUsage()) {texture=&DotMask_01}
  n19 = NamedRerouteUsage()
  n20 = StaticSwitch(True=Multiply(A=@n18.RGBA, B=Multiply(B=@n19) {constA=20.48, desc=20.48}), False=@n18.RGBA, Value=NamedRerouteUsage())
  n21 = BumpOffset(Coordinate=@n17, Height=NamedRerouteUsage())
  n22 = TextureSample(UVs=StaticSwitch(True=Rotator(Coordinate=@n21), False=@n21, Value=NamedRerouteUsage()), Tex=NamedRerouteUsage()) {texture=&DotMask_01}
  n23 = NamedRerouteUsage()
  n24 = StaticSwitch(True=Multiply(A=@n22.RGBA, B=Multiply(B=@n23) {constA=10.24, desc="10,24"}), False=@n22.RGBA, Value=NamedRerouteUsage())
  n25 = BumpOffset(Coordinate=@n21, Height=NamedRerouteUsage())
  n26 = TextureSample(UVs=StaticSwitch(True=Rotator(Coordinate=@n25), False=@n25, Value=NamedRerouteUsage()), Tex=NamedRerouteUsage()) {texture=&DotMask_01}
  n27 = NamedRerouteUsage()
  n28 = StaticSwitch(True=Multiply(A=@n26.RGBA, B=Multiply(B=@n27) {constA=5.12, desc=5.12}), False=@n26.RGBA, Value=NamedRerouteUsage())
  n29 = BumpOffset(Coordinate=@n25, Height=NamedRerouteUsage())
  n30 = TextureSample(UVs=StaticSwitch(True=Rotator(Coordinate=@n29), False=@n29, Value=NamedRerouteUsage()), Tex=NamedRerouteUsage()) {texture=&DotMask_01}
  n31 = NamedRerouteUsage()
  n32 = StaticSwitch(True=Multiply(A=@n30.RGBA, B=Multiply(B=@n31) {constA=2.56, desc=2.56}), False=@n30.RGBA, Value=NamedRerouteUsage())
  n33 = Add(A=Add(A=Add(A=@n4, B=@n8), B=Add(A=@n12, B=@n16)), B=Add(A=Add(A=@n20, B=@n24), B=Add(A=@n28, B=@n32)))
  n34 = BumpOffset(Coordinate=@n29, Height=NamedRerouteUsage())
  n35 = TextureSample(UVs=StaticSwitch(True=Rotator(Coordinate=@n34), False=@n34, Value=NamedRerouteUsage()), Tex=NamedRerouteUsage()) {texture=&DotMask_01}
  n36 = NamedRerouteUsage()
  n37 = StaticSwitch(True=Multiply(A=@n35.RGBA, B=Multiply(B=@n36) {constA=1.28, desc=1.28}), False=@n35.RGBA, Value=NamedRerouteUsage())
  n38 = BumpOffset(Coordinate=@n34, Height=NamedRerouteUsage())
  n39 = TextureSample(UVs=StaticSwitch(True=Rotator(Coordinate=@n38), False=@n38, Value=NamedRerouteUsage()), Tex=NamedRerouteUsage()) {texture=&DotMask_01}
  n40 = NamedRerouteUsage()
  n41 = StaticSwitch(True=Multiply(A=@n39.RGBA, B=Multiply(B=@n40) {constA=0.64, desc=0.64}), False=@n39.RGBA, Value=NamedRerouteUsage())
  n42 = BumpOffset(Coordinate=@n38, Height=NamedRerouteUsage())
  n43 = TextureSample(UVs=StaticSwitch(True=Rotator(Coordinate=@n42), False=@n42, Value=NamedRerouteUsage()), Tex=NamedRerouteUsage()) {texture=&DotMask_01}
  n44 = NamedRerouteUsage()
  n45 = StaticSwitch(True=Multiply(A=@n43.RGBA, B=Multiply(B=@n44) {constA=0.32, desc=0.32}), False=@n43.RGBA, Value=NamedRerouteUsage())
  n46 = BumpOffset(Coordinate=@n42, Height=NamedRerouteUsage())
  n47 = TextureSample(UVs=StaticSwitch(True=Rotator(Coordinate=@n46), False=@n46, Value=NamedRerouteUsage()), Tex=NamedRerouteUsage()) {texture=&DotMask_01}
  n48 = NamedRerouteUsage()
  n49 = StaticSwitch(True=Multiply(A=@n47.RGBA, B=Multiply(B=@n48) {constA=0.16, desc=0.16}), False=@n47.RGBA, Value=NamedRerouteUsage())
  n50 = BumpOffset(Coordinate=@n46, Height=NamedRerouteUsage())
  n51 = TextureSample(UVs=StaticSwitch(True=Rotator(Coordinate=@n50), False=@n50, Value=NamedRerouteUsage()), Tex=NamedRerouteUsage()) {texture=&DotMask_01}
  n52 = NamedRerouteUsage()
  n53 = StaticSwitch(True=Multiply(A=@n51.RGBA, B=Multiply(B=@n52) {constA=0.08, desc=0.08}), False=@n51.RGBA, Value=NamedRerouteUsage())
  n54 = BumpOffset(Coordinate=@n50, Height=NamedRerouteUsage())
  n55 = TextureSample(UVs=StaticSwitch(True=Rotator(Coordinate=@n54), False=@n54, Value=NamedRerouteUsage()), Tex=NamedRerouteUsage()) {texture=&DotMask_01}
  n56 = NamedRerouteUsage()
  n57 = StaticSwitch(True=Multiply(A=@n55.RGBA, B=Multiply(B=@n56) {constA=0.04, desc=0.04}), False=@n55.RGBA, Value=NamedRerouteUsage())
  n58 = BumpOffset(Coordinate=@n54, Height=NamedRerouteUsage())
  n59 = TextureSample(UVs=StaticSwitch(True=Rotator(Coordinate=@n58), False=@n58, Value=NamedRerouteUsage()), Tex=NamedRerouteUsage()) {texture=&DotMask_01}
  n60 = NamedRerouteUsage()
  n61 = StaticSwitch(True=Multiply(A=@n59.RGBA, B=Multiply(B=@n60) {constA=0.02, desc=0.02}), False=@n59.RGBA, Value=NamedRerouteUsage())
  n62 = BumpOffset(Coordinate=@n58, Height=NamedRerouteUsage())
  n63 = TextureSample(UVs=StaticSwitch(True=Rotator(Coordinate=@n62), False=@n62, Value=NamedRerouteUsage()), Tex=NamedRerouteUsage()) {texture=&DotMask_01}
  n64 = NamedRerouteUsage()
  n65 = StaticSwitch(True=Multiply(A=@n63.RGBA, B=Multiply(B=@n64) {constA=0.01, desc=0.01}), False=@n63.RGBA, Value=NamedRerouteUsage())
  n66 = Add(A=Add(A=Add(A=@n37, B=@n41), B=Add(A=@n45, B=@n49)), B=Add(A=Add(A=@n53, B=@n57), B=Add(A=@n61, B=@n65)))
  FunctionOutput(None=Clamp(None=Add(A=@n33, B=@n66)))
  NamedRerouteDeclaration(None=FunctionInput(Preview=StaticBool()) {bUsePreviewValueAsDefault=true, inputName=RotateUVs, inputType=FunctionInput_StaticBool, sortPriority=3}) {name=Rotate_UVs, nodeColor=<rgba 0.4062 1 0 1>}
  NamedRerouteDeclaration(None=FunctionInput(Preview=TextureObject() {texture=&DotMask_01}) {bUsePreviewValueAsDefault=true, inputName=Texture, inputType=FunctionInput_Texture2D, sortPriority=-1}) {name=Texture, nodeColor=<rgba 0.4062 1 0 1>}
  NamedRerouteDeclaration(None=FunctionInput(Preview=Constant() {r=-1.5}) {bUsePreviewValueAsDefault=true, inputName=Offset, inputType=FunctionInput_Scalar, sortPriority=1}) {name=Offset, nodeColor=<rgba 0.4062 1 0 1>}
  NamedRerouteDeclaration(None=FunctionInput(Preview=Constant() {r=0.05}) {bUsePreviewValueAsDefault=true, inputName=GradientDarkness, inputType=FunctionInput_Scalar, sortPriority=5}) {name=GradientDarkness, nodeColor=<rgba 0.4062 1 0 1>}
  NamedRerouteDeclaration(None=FunctionInput(Preview=StaticBool() {value=true}) {bUsePreviewValueAsDefault=true, inputName=GradientFalloff, inputType=FunctionInput_StaticBool, sortPriority=4}) {name=GradientFalloff, nodeColor=<rgba 0.4062 1 0 1>}
  NamedRerouteDeclaration(None=FunctionInput(Preview=TextureCoordinate()) {bUsePreviewValueAsDefault=true, inputName=UVs, inputType=FunctionInput_Vector2, sortPriority=2}) {name=UVs, nodeColor=<rgba 0.4062 1 0 1>}
 ]}
