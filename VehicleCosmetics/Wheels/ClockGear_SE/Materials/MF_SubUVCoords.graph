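# Material function in the ClockGear_SE wheel cosmetic that maps UVs to one cell of a flipbook sheet for a given Frame.
# Frame drops its fraction, then fmod and floor against the SubImages count give the column and row.
# That cell offset is added to the UVs, which default to the texture coordinate, and scaled by the reciprocal of SubImages, a scalar input.
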MaterialFunction /VehicleCosmetics/Wheels/ClockGear_SE/Materials/MF_SubUVCoords
graph {nodes [
  n1 = FunctionInput() {inputName=Frame, inputType=FunctionInput_Scalar, previewValue=<1 0 0 1>}
  n2 = Subtract(A=@n1, B=Frac(None=@n1))
  n3 = FunctionInput() {inputName=SubImages, inputType=FunctionInput_Scalar, previewValue=<1 1 0 1>}
  n4 = Divide(A=Constant2Vector() {g=1, r=1}, B=@n3)
  n5 = ComponentMask(None=@n4) {r=true}
  n6 = AppendVector(A=Fmod(A=@n2, B=ComponentMask(None=@n3) {r=true}), B=Floor(None=Multiply(A=@n2, B=@n5)))
  n7 = FunctionInput(Preview=TextureCoordinate()) {bUsePreviewValueAsDefault=true, inputName=UVs, inputType=FunctionInput_Vector2}
  FunctionOutput(None=Multiply(A=Add(A=@n6, B=@n7), B=@n4))
 ]}
