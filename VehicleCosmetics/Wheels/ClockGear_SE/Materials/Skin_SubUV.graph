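# Material function for the ClockGear_SE wheel cosmetic that plays a sub-image flipbook texture, outputting color and alpha.
# Time times Speed wraps through Frac, adds OffsetFramePosition, and scales by the SubImages tile count to drive MF_SubUVCoords.
# A static switch optionally lerps between the current and next frame.
# That lerp's alpha is Frac of Speed alone, so the blend weight does not follow the animated frame position.
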MaterialFunction /VehicleCosmetics/Wheels/ClockGear_SE/Materials/Skin_SubUV
graph {
 properties {description=RGB}
 nodes [
  n1 = FunctionInput() {bUsePreviewValueAsDefault=true, inputName=Speed, inputType=FunctionInput_Scalar, sortPriority=2}
  n2 = Multiply(A=Time(), B=@n1)
  n3 = FunctionInput() {bUsePreviewValueAsDefault=true, inputName=SubImages, inputType=FunctionInput_Vector2, sortPriority=1, previewValue=<1 1 0 1>}
  n4 = Multiply(
   A=Add(
    A=Frac(None=@n2)
    B=FunctionInput() {bUsePreviewValueAsDefault=true, inputName=OffsetFramePosition, inputType=FunctionInput_Scalar, sortPriority=4}
   )
   B=Multiply(A=ComponentMask(None=@n3) {r=true}, B=ComponentMask(None=@n3) {g=true})
  )
  n5 = FunctionInput(Preview=TextureCoordinate()) {bUsePreviewValueAsDefault=true, inputName=UVs, inputType=FunctionInput_Vector2, sortPriority=3}
  n6 = MaterialFunctionCall!(SubImages=@n3, UVs=@n5, Frame=@n4) {materialFunction=&/MF_SubUVCoords}
  n7 = FunctionInput(Preview=TextureObject()) {bUsePreviewValueAsDefault=true, inputName=Texture, inputType=FunctionInput_Texture2D, previewValue=<1 1 0 1>}
  n8 = TextureSample(UVs=@n6.Result, Tex=@n7) {samplerSource=SSM_Wrap_WorldGroupSettings}
  n9 = TextureSample(
   UVs=MaterialFunctionCall(SubImages=@n3, UVs=@n5, Frame=Add(A=@n4)) {materialFunction=&MF_SubUVCoords}.Result
   Tex=@n7
  ) {samplerSource=SSM_Wrap_WorldGroupSettings}
  n10 = Frac(None=@n1)
  n11 = TextureSample(UVs=@n6.Result, Tex=@n7) {samplerSource=SSM_Wrap_WorldGroupSettings}
  n12 = FunctionInput(Preview=StaticBool()) {bUsePreviewValueAsDefault=true, inputName="LinearBlend(MoreExpensive)", inputType=FunctionInput_StaticBool, sortPriority=5}
  FunctionOutput(None=StaticSwitch(True=LinearInterpolate(A=@n8.RGBA, B=@n9.RGBA, Alpha=@n10), False=@n11.RGBA, Value=@n12)) {outputName=RGB}
  FunctionOutput(None=StaticSwitch(True=LinearInterpolate(A=@n8.A, B=@n9.A, Alpha=@n10), False=@n11.A, Value=@n12)) {outputName=Alpha, sortPriority=1}
 ]
}
