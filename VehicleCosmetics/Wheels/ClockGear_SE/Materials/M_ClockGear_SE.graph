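# Masked material for the ClockGear SE vehicle wheel cosmetic, drawing an animated, tinted emissive gear effect inside a circular cutout.
# Green-channel flipbook samples from a UV1 SubUV layer and a time-driven UV0 infinity-offset layer are summed, tinted by Custom Color and faded radially.
# A dot product of centered UVs drives both that falloff and the hard opacity mask.
# The result passes through shared vehicle damage and de-rez functions.
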Material /VehicleCosmetics/Wheels/ClockGear_SE/Materials/M_ClockGear_SE
graph {
 properties {
  bAutomaticallySetUsageInEditor=false
  bUseMaterialAttributes=true
  bUsedWithMeshDeformer=true
  bUsedWithSkeletalMesh=true
  blendMode=BLEND_Masked
  usedShadingModels=Default Lit
 }
 nodes [
  n1 = VectorParameter() {parameterName=Custom Color, defaultValue=<rgba 0.0296 0 0.4852 1>}
  n2 = ScalarParameter() {defaultValue=4, parameterName=ColorMultiply}
  n3 = TextureObjectParameter() {parameterName=Main Tex, texture=/VehicleCosmetics/Wheels/ClockGear_SE/Textures/T_ClockGearSE_Flipbook_M}
  n4 = ScalarParameter() {defaultValue=4, parameterName=NumberOfRows}
  n5 = ScalarParameter() {defaultValue=4, parameterName=NumberOfColums}
  n6 = Multiply(A=TextureCoordinate() {coordinateIndex=1}, B=ScalarParameter() {defaultValue=1, parameterName=Tiling_Layer_1})
  n7 = ComponentMask(None=MaterialFunctionCall!(
    Texture=@n3
    SubImages=AppendVector(A=@n4, B=@n5)
    Speed=ScalarParameter() {defaultValue=-1, parameterName=Speed_Layer1}
    UVs=Frac(None=@n6)
    OffsetFramePosition
    "LinearBlend(MoreExpensive)"=StaticBool() {value=true}
   ) {materialFunction=&/Skin_SubUV}.RGB) {g=true}
  n8 = ScalarParameter() {defaultValue=-1, parameterName=BumpOffsetAmount}
  n9 = Constant2Vector() {g=4, r=4}
  n10 = Frac(None=Multiply(A=Time()))
  n11 = Multiply(A=@n9.R, B=@n9.R)
  n12 = MaterialFunctionCall!(SubImages=@n9, UVs=TextureCoordinate(), Frame=Add(A=Multiply(A=@n10, B=@n11))) {materialFunction=&/MF_SubUVCoords}
  n13 = StaticBool()
  n14 = StaticBool() {value=true}
  n15 = Constant() {r=0.001}
  n16 = Add(
   A=@n7
   B=ComponentMask(None=MaterialFunctionCall!(Texture=@n3, Offset=@n8, UVs=@n12.Result, RotateUVs=@n13, GradientFalloff=@n14, GradientDarkness=@n15) {materialFunction=&/MF_Infinity_OffsetFunction}.Result) {g=true}
  )
  n17 = Add(A=TextureCoordinate()) {constB=-0.5}
  n18 = DotProduct(A=@n17, B=@n17)
  n19 = Power(Base=OneMinus(None=Clamp(None=@n18))) {constExponent=12}
  n20 = Add(A=Multiply(A=@n7) {constB=0.05}, B=Multiply(A=Multiply(A=@n1.RGB, B=@n2), B=Multiply(A=@n16, B=@n19)))
  n21 = OneMinus(None=@n18)
  n22 = Clamp(None=Ceil(None=Subtract(A=@n21) {constB=0.78}))
  MaterialFunctionCall(In=MaterialFunctionCall(Material=SetMaterialAttributes!(MaterialAttributes, Emissive Color=@n20, Opacity Mask=@n22) {attributeSetTypes [#1, #2]}) {materialFunction=&MF_BaseVehicle_Damage}.Result) {materialFunction=&MF_VehicleDeRez_Transition}
 ]
}
